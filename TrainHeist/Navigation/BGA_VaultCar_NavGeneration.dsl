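; Train heist vault car actor that runs an AI reinforcement encounter through its living world volume child actor.
; After a short delay, the server enables the point provider and starts BP_DynamicReinforcementsManager; clients destroy the child actor component.
; One-shot timers later toggle the provider off and on to refresh it, then pause reinforcements.
; When the encounter's last actor dies, it broadcasts OnEncounterCompleted.
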(event EventBeginPlay
  (bind _livingworldmanager_childactor (Variables|Default|GetLivingWorldManager_ChildActor))
  (bind _childactor (Variables|/Script/Engine.ChildActorComponent|GetChildActor _livingworldmanager_childactor))
  (Networking|SwitchHasAuthority
    (:Authority
      (Utilities|FlowControl|Delay 0.4)
      (Class|BPDynamicReinforcementsManager|SetEncounterReservoirOwner (Variables|Default|GetBP_DynamicReinforcementsManager) (Actor|GetOwner))
      (Class|FortAthenaLivingWorldPointProviderInterface|EnablePointProvider_DEPRECATED(Message) _childactor)
      (bind _asfort_athena_living_world_volume (Utilities|Casting|CastToFortAthenaLivingWorldVolume _childactor))
      (Class|BPDynamicReinforcementsManager|StartReinforcements (Variables|Default|GetBP_DynamicReinforcementsManager) _asfort_athena_living_world_volume 0 0)
      (Utilities|Time|SetTimerbyEvent (EventDispatchers|CreateEvent :!event Custom|UpdatePointProviders) 6.0)
      (Utilities|Time|SetTimerbyEvent (EventDispatchers|CreateEvent :!event Custom|StopReinforcements) 45.0)
      (bind _bp_dynamicreinforcementsmanager (Variables|Default|GetBP_DynamicReinforcementsManager)
        (:then
          (bind _encounterref (Variables|/VividRazor_AIConvoy/Blueprints/Components/BP_DynamicReinforcementsManager.BP_DynamicReinforcementsManager_C|GetEncounterRef _bp_dynamicreinforcementsmanager)
            (:then
              (EventDispatchers|BindEventtoOnLastActorDied _encounterref (delegate Custom|Last%20Actor%20Died)))
            (:else)))
        (:else)))
    (:Remote
      (Components|DestroyComponent (Variables|Default|GetLivingWorldManager_ChildActor)))))

(event Custom|Last%20Actor%20Died
  (params (EncounterInstance :type Fort%20Athena%20Living%20World%20Encounter%20Instance%20Object%20Reference))
  (Default|CallOnEncounterCompleted))

(event Custom|UpdatePointProviders
  (bind _livingworldmanager_childactor (Variables|Default|GetLivingWorldManager_ChildActor))
  (bind _childactor (Variables|/Script/Engine.ChildActorComponent|GetChildActor _livingworldmanager_childactor))
  (Class|FortAthenaLivingWorldPointProviderInterface|DisablePointProvider(Message) _childactor)
  (Utilities|FlowControl|Delay 0.2)
  (Class|FortAthenaLivingWorldPointProviderInterface|EnablePointProvider_DEPRECATED(Message) _childactor))

(event Custom|StopReinforcements
  (Class|BPDynamicReinforcementsManager|PauseReinforcements (Variables|Default|GetBP_DynamicReinforcementsManager)))

(fn ConstructionScript)
